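# Material function for the SkyMango fragment prop that outputs a world position offset jitter (WPOJitter) relative to a ZeroPointPosition.
# Layered sines of time, phased by particle random or pivot distance via the MeshParticle switch, drive movement along the direction away from that point.
# A three-axis sine shake fades as that radial wave rises.
# The rift orb parameter collection supplies jitter strength, radial offset and added time.
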MaterialFunction /DragonCartNarrative/Props/SkyMango/SkyMango_Fragment_A/Materials/MF_Nex_Core_WPO
graph {expressions {
  MaterialExpressionAdd_0 {
   class=MaterialExpressionAdd
   inputs [
    A {from=MaterialExpressionMultiply_7}
    B {from=MaterialExpressionCollectionParameter_1}
   ]
  }
  MaterialExpressionAdd_1 {
   class=MaterialExpressionAdd
   inputs [
    A {from=MaterialExpressionConstant3Vector_0}
    B {from=MaterialExpressionSubtract_0}
   ]
  }
  MaterialExpressionAdd_2 {
   class=MaterialExpressionAdd
   inputs [
    A {from=MaterialExpressionStaticSwitchParameter_1}
    B {from=MaterialExpressionAdd_6}
   ]
  }
  MaterialExpressionAdd_3 {
   class=MaterialExpressionAdd
   inputs [
    A {from=MaterialExpressionMultiply_1}
    B {from=MaterialExpressionMultiply_0}
   ]
  }
  MaterialExpressionAdd_4 {
   class=MaterialExpressionAdd
   inputs [
    A {from=MaterialExpressionAdd_3}
    B {from=MaterialExpressionMultiply_2}
   ]
  }
  MaterialExpressionAdd_5 {
   class=MaterialExpressionAdd
   inputs [
    A {from=MaterialExpressionMultiply_9}
    B {from=MaterialExpressionMultiply_5}
   ]
  }
  MaterialExpressionAdd_6 {
   class=MaterialExpressionAdd
   inputs [
    A {from=MaterialExpressionTime_2}
    B {from=MaterialExpressionCollectionParameter_2}
   ]
  }
  MaterialExpressionCollectionParameter_0 {
   class=MaterialExpressionCollectionParameter
   inputs []
   properties {collection=/DragonCartNarrative/Assets/VFX/RiftPortalIdle/MPC_DragonCart_Orb_Rift, parameterName=ZPJitter}
  }
  MaterialExpressionCollectionParameter_1 {
   class=MaterialExpressionCollectionParameter
   inputs []
   properties {collection=&MPC_DragonCart_Orb_Rift, parameterName=ZPOffset}
  }
  MaterialExpressionCollectionParameter_2 {
   class=MaterialExpressionCollectionParameter
   inputs []
   properties {collection=&MPC_DragonCart_Orb_Rift, parameterName=ZPJitterTimeAdd}
  }
  MaterialExpressionConstant3Vector_0 {class=MaterialExpressionConstant3Vector, properties {constant=<rgba 0 0 0.001 0>}}
  MaterialExpressionConvert_0 {
   class=MaterialExpressionConvert
   inputs [
    None {from=MaterialExpressionSine_4}
    None {from=MaterialExpressionSine_5}
    None {from=MaterialExpressionSine_6}
   ]
   properties {
    nodeName=Make Vector3
    convertInputs [
     {
      type=Scalar
      expressionInput {
       expression=&self:MaterialExpressionSine_4
       inputName=None
       mask=0
       maskA=0
       maskB=0
       maskG=0
       maskR=0
       outputIndex=0
      }
     }
     {
      type=Scalar
      expressionInput {
       expression=&self:MaterialExpressionSine_5
       inputName=None
       mask=0
       maskA=0
       maskB=0
       maskG=0
       maskR=0
       outputIndex=0
      }
     }
     {
      type=Scalar
      expressionInput {
       expression=&self:MaterialExpressionSine_6
       inputName=None
       mask=0
       maskA=0
       maskB=0
       maskG=0
       maskR=0
       outputIndex=0
      }
     }
    ]
    convertOutputs [{type=Vector3}]
   }
  }
  MaterialExpressionDivide_0 {
   class=MaterialExpressionDivide
   inputs [
    A {from=MaterialExpressionLength_0}
    B {from=""}
   ]
   properties {constB=10}
  }
  MaterialExpressionFunctionOutput_1 {class=MaterialExpressionFunctionOutput, inputs [None {from=MaterialExpressionAdd_5, output=""}], properties {outputName=WPOJitter}}
  MaterialExpressionLength_0 {class=MaterialExpressionLength, inputs [None {from=MaterialExpressionStaticSwitchParameter_0, output=""}], properties {}}
  MaterialExpressionLinearInterpolate_1 {
   class=MaterialExpressionLinearInterpolate
   inputs [
    A {from=MaterialExpressionMultiply_3}
    B {from=""}
    Alpha {from=MaterialExpressionSaturate_0}
   ]
   properties {constB=0}
  }
  MaterialExpressionMaterialFunctionCall_1 {
   class=MaterialExpressionMaterialFunctionCall
   inputs []
   properties {materialFunction=&ObjectPivotPoint}
  }
  MaterialExpressionMultiply_0 {
   class=MaterialExpressionMultiply
   inputs [
    A {from=MaterialExpressionSine_1}
    B {from=""}
   ]
   properties {constB=0.15}
  }
  MaterialExpressionMultiply_1 {
   class=MaterialExpressionMultiply
   inputs [
    A {from=MaterialExpressionSine_3}
    B {from=""}
   ]
   properties {constB=0.4}
  }
  MaterialExpressionMultiply_2 {
   class=MaterialExpressionMultiply
   inputs [
    A {from=MaterialExpressionSine_2}
    B {from=""}
   ]
   properties {constB=0.7}
  }
  MaterialExpressionMultiply_3 {
   class=MaterialExpressionMultiply
   inputs [
    A {from=MaterialExpressionConvert_0}
    B {from=MaterialExpressionScalarParameter_2}
   ]
  }
  MaterialExpressionMultiply_5 {
   class=MaterialExpressionMultiply
   inputs [
    A {from=MaterialExpressionNormalize_0}
    B {from=MaterialExpressionAdd_0}
   ]
  }
  MaterialExpressionMultiply_7 {
   class=MaterialExpressionMultiply
   inputs [
    A {from=MaterialExpressionAdd_4}
    B {from=MaterialExpressionCollectionParameter_0}
   ]
  }
  MaterialExpressionMultiply_9 {
   class=MaterialExpressionMultiply
   inputs [
    A {from=MaterialExpressionLinearInterpolate_1}
    B {from=MaterialExpressionCollectionParameter_0}
   ]
  }
  MaterialExpressionNormalize_0 {class=MaterialExpressionNormalize, inputs [VectorInput {from=MaterialExpressionAdd_1, output=""}]}
  MaterialExpressionParticlePositionWS_0 {class=MaterialExpressionParticlePositionWS, inputs [], properties {}}
  MaterialExpressionParticleRandom_0 {class=MaterialExpressionParticleRandom, inputs [], properties {}}
  MaterialExpressionReroute_0 {class=MaterialExpressionReroute, inputs [None {from=MaterialExpressionAdd_2}]}
  MaterialExpressionSaturate_0 {class=MaterialExpressionSaturate, inputs [None {from=MaterialExpressionAdd_4}]}
  MaterialExpressionScalarParameter_2 {class=MaterialExpressionScalarParameter, properties {defaultValue=2}}
  MaterialExpressionSine_1 {class=MaterialExpressionSine, inputs [None {from=MaterialExpressionAdd_2}], properties {period=0.22}}
  MaterialExpressionSine_2 {class=MaterialExpressionSine, inputs [None {from=MaterialExpressionAdd_2}], properties {period=0.95}}
  MaterialExpressionSine_3 {class=MaterialExpressionSine, inputs [None {from=MaterialExpressionAdd_2}], properties {period=0.33}}
  MaterialExpressionSine_4 {class=MaterialExpressionSine, inputs [None {from=MaterialExpressionReroute_0}], properties {period=0.22}}
  MaterialExpressionSine_5 {class=MaterialExpressionSine, inputs [None {from=MaterialExpressionReroute_0}], properties {period=0.19}}
  MaterialExpressionSine_6 {class=MaterialExpressionSine, inputs [None {from=MaterialExpressionReroute_0}], properties {period=0.175}}
  MaterialExpressionStaticSwitchParameter_0 {
   class=MaterialExpressionStaticSwitchParameter
   inputs [
    True {from=MaterialExpressionParticlePositionWS_0, output=""}
    False {from=MaterialExpressionMaterialFunctionCall_1, output=Object Pivot Location}
   ]
   properties {parameterName=MeshParticle}
  }
  MaterialExpressionStaticSwitchParameter_1 {
   class=MaterialExpressionStaticSwitchParameter
   inputs [
    True {from=MaterialExpressionParticleRandom_0, output=""}
    False {from=MaterialExpressionDivide_0, output=""}
   ]
   properties {parameterName=MeshParticle}
  }
  MaterialExpressionSubtract_0 {
   class=MaterialExpressionSubtract
   inputs [
    A {from=MaterialExpressionStaticSwitchParameter_0}
    B {from=MaterialExpressionVectorParameter_7, output=RGB}
   ]
  }
  MaterialExpressionTime_2 {class=MaterialExpressionTime}
  MaterialExpressionVectorParameter_7 {class=MaterialExpressionVectorParameter, properties {parameterName=ZeroPointPosition, defaultValue=<rgba -11011 0 3936.7444 0>}}
 }}
