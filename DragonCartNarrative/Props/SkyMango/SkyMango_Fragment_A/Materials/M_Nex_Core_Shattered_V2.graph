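# Default Lit material for a shattered, glowing glass Nexus core on the Sky Mango fragment prop, usable with Nanite and Niagara.
# World position offset combines bobbing and Z rotation, with matching normal rotation, alongside MF_Nex_Core_WPO jitter and collection-driven shatter spread.
# Emissive sums crack glow and inner and outer Fresnel terms.
# Layered sines flicker the Fresnel, sped up by Destabilize (parameter or MPC), becoming constant on low quality and mobile.
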Material /DragonCartNarrative/Props/SkyMango/SkyMango_Fragment_A/Materials/M_Nex_Core_Shattered_V2
graph {
 properties {
  bUsedWithNanite=true
  bUsedWithNiagaraMeshParticles=true
  bUsedWithParticleSprites=true
  pixelDepthOffsetMode=PDOM_Legacy
  usedShadingModels=Default Lit
 }
 nodes [
  Add(
   A=NamedRerouteUsage()
   B=MaterialFunctionCall() {materialFunction=&/MF_Nex_Core_WPO}.WPOJitter
  )
  Constant() {r=1}
  n1 = MaterialFunctionCall!(
   Time Multiplier=CollectionParameter() {collection=/Game/Athena/Environments/Nexus/Materials/NexCollection, parameterName=TimeMultiplier}
   Lerp Shatter Pulse=CollectionParameter() {collection=&NexCollection, parameterName=PulseLerp}
   Shake=CollectionParameter() {collection=&NexCollection, parameterName=Shake}
   Progress=CollectionParameter() {collection=&NexCollection, parameterName=Progress}
  ) {materialFunction=/Game/Athena/Environments/Nexus/Materials/MF_Nex_Core_WPO}
  LinearInterpolate(B=@n1.WPOSpread, Alpha=ScalarParameter() {defaultValue=1, parameterName=Use WPO})
  n2 = Multiply(A=TextureCoordinate() {uTiling=2, vTiling=2}, B=NamedRerouteUsage())
  n3 = Multiply(A=NamedRerouteUsage(), B=TextureCoordinate())
  n4 = MaterialFunctionCall(
   Normal=LinearInterpolate(
    A=TextureSample(UVs=@n2) {samplerType=SAMPLERTYPE_Normal, texture=/Hera/Props/SkyMango/SkyMango_Fragment_A/Textures/T_Glass_Cracks_N2}.RGB
    B=TextureSample(UVs=@n3) {samplerType=SAMPLERTYPE_Normal, texture=&T_Glass_Cracks_N2}.RGB
   )
   Flatness=ScalarParameter() {defaultValue=1, parameterName=Normal Flatness}
  ) {materialFunction=&FlattenNormal}
  n5 = Add(
   A=Multiply(A=Time(), B=ScalarParameter() {defaultValue=0.025, group=WPO, parameterName=Rotation Speed})
   B=ScalarParameter() {defaultValue=-0.01, group=WPO, parameterName=Rotation Angle Min}
  )
  LinearInterpolate(
   A=@n4.Result
   B=MaterialFunctionCall(Rotation Axis=Constant3Vector() {constant=<rgba 0 0 1 0>}, Rotation Angle=@n5, World Space Vertex Normals=Transform(None=@n4.Result)) {materialFunction=&FixRotateAboutAxisNormals}.New Tangent Space Vertex Normal
   Alpha=Saturate(None=ScalarParameter() {defaultValue=1, parameterName="Use Rotation WPO? (Switch)", sliderMax=1})
  )
  n6 = VectorParameter() {parameterName=CrackGlow, defaultValue=<rgba 2.4794 9.159 16 1>}
  n7 = NamedRerouteUsage()
  n8 = Multiply(A=TextureCoordinate() {uTiling=2, vTiling=2}, B=@n7)
  n9 = Multiply(A=@n7, B=TextureCoordinate())
  n10 = Power(Base=LinearInterpolate(
    A=TextureSample(UVs=@n8) {texture=/Hera/Props/SkyMango/SkyMango_Fragment_A/Textures/T_Glass_Cracks_M}.G
    B=TextureSample(UVs=@n9) {texture=&T_Glass_Cracks_M}.G
   ) {constAlpha=0.7})
  n11 = ScalarParameter() {defaultValue=2, parameterName=Outer Shell Fresnel}
  n12 = ScalarParameter() {defaultValue=0.8, parameterName=FlickerSpeed}
  n13 = CollectionParameter() {collection=/Game/Athena/Environments/Nexus/Materials/FXshader/MPC_Nexus, parameterName=Destabilize}
  n14 = ScalarParameter() {parameterName=Destabilize}
  n15 = Multiply(
   A=Time()
   B=LinearInterpolate(A=@n12, B=Multiply(A=@n12) {constB=2}, Alpha=StaticSwitchParameter(True=@n13, False=@n14) {parameterName=UseMPC})
  )
  n16 = Sine(None=@n15) {period=0.55}
  n17 = Sine(None=@n15) {period=0.11}
  n18 = Add(A=Add(A=Multiply(B=@n16), B=Multiply(A=@n17) {constB=0.2}), B=Multiply(A=Sine(None=@n15) {period=0.175}) {constB=0.3})
  n19 = ConstantBiasScale(None=Multiply(A=Sine(None=@n15) {period=3.25}, B=Sine(None=@n15) {period=2.325}))
  n20 = Constant() {r=0.6}
  n21 = ShadingPathSwitch(Default=QualitySwitch(Default=Multiply(A=@n18, B=@n19), Low=@n20), Mobile=@n20) {desc=Migrated from FeatureLevelSwitch}
  n22 = Multiply(A=@n21) {constB=0.01}
  n23 = Multiply(
   A=VectorParameter() {parameterName=Outer Fresnel Color, defaultValue=<rgba 0.1073 0.0986 0.25 1>}.RGB
   B=Fresnel(ExponentIn=Add(A=@n11, B=@n22)) {baseReflectFraction=0}
  )
  n24 = Add(A=@n22, B=ScalarParameter() {defaultValue=0.4, parameterName=Inner Glow Fresnel})
  n25 = Power(
   Base=OneMinus(None=Fresnel(ExponentIn=@n24) {baseReflectFraction=0})
   Exp=Add(A=@n22, B=ScalarParameter() {defaultValue=4, parameterName=Inner Fresnel Power})
  )
  n26 = VectorParameter() {parameterName=Inner Fresnel Color, defaultValue=<rgba 0.8023 1.9068 3 1>}
  n27 = Multiply(A=LinearInterpolate(Alpha=@n21) {constA=0.7}, B=Multiply(A=Multiply(A=@n25, B=@n26.RGB)))
  MaterialFunctionCall(Emissive=Add(A=Multiply(A=@n6.RGB, B=@n10), B=Add(A=@n23, B=@n27))) {materialFunction=&MF_FX_EmissiveAdjustments}
  MaterialProxyReplace(Realtime=Constant(), MaterialProxy=Constant3Vector() {constant=<rgba 0.1529 0.451 0.5922 1>})
  n28 = VectorParameter() {parameterName=CrackTiling, defaultValue=<rgba 1 1 0 1>}
  NamedRerouteDeclaration(None=AppendVector(A=@n28.R, B=@n28.R)) {name=CrackTiling, nodeColor=<rgba 0 0.4844 1 1>}
  n29 = VectorParameter() {parameterName=EmissiveTiling, defaultValue=<rgba 1 1 0 0>}
  NamedRerouteDeclaration(None=AppendVector(A=@n29.R, B=@n29.R)) {name=EmissiveTiling, nodeColor=<rgba 0 0.6719 1 1>}
  n30 = Time()
  n31 = ScalarParameter() {defaultValue=0.4, group=WPO, parameterName=Bob Sine Speed}
  n32 = LinearInterpolate(
   A=VectorParameter() {group=WPO, parameterName=BobHeightMax, defaultValue=<rgba 0 0 4 0>}.RGB
   B=VectorParameter() {group=WPO, parameterName=BobHeightMin, defaultValue=<rgba 0 0 -4 0>}.RGB
   Alpha=Sine(None=Multiply(A=@n30, B=@n31))
  )
  n33 = RotateAboutAxis(
   NormalizedRotationAxis=Constant3Vector() {constant=<rgba 0 0 1 0>}
   RotationAngle=@n5
   PivotPoint=ObjectPositionWS() {originType=CameraRelative}
   Position=WorldPosition() {worldPositionShaderOffset=WPT_CameraRelative}.XYZ
  )
  n34 = ScalarParameter() {defaultValue=1, parameterName="Use Rotation WPO? (Switch)", sliderMax=1}
  NamedRerouteDeclaration(None=LinearInterpolate(B=Add(A=@n32, B=@n33), Alpha=Saturate(None=@n34))) {name=WPO, nodeColor=<rgba 1 0 0.8203 1>}
  Saturate(None=LinearInterpolate(Alpha=@n1.Mask) {constA=1, constB=-0.5})
  ScalarParameter() {defaultValue=0.02, parameterName=Roughness}
  TextureSample() {samplerType=SAMPLERTYPE_Normal, texture=/Game/Effects/Fort_Effects/Textures/Tile/T_Glass_Cracks_N}
  TextureSample() {samplerType=SAMPLERTYPE_Normal, texture=&T_Glass_Cracks_N}
 ]
}
